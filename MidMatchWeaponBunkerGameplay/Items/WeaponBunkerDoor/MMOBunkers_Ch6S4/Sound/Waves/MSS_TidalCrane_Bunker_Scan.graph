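# Stereo one-shot MetaSound Source for the TidalCrane weapon bunker door scan, set to the player-interaction sound class.
# A single Wave Player starts the TidalCrane_Door_Activation wave on play, feeds left and right outputs directly, and signals finished when the wave ends.
# The attenuation distance input is unused in the graph; spatial falloff is assigned through a scanner attenuation asset, and a concurrency override limits simultaneous instances.
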MetaSoundSource /MidMatchWeaponBunkerGameplay/Items/WeaponBunkerDoor/MMOBunkers_Ch6S4/Sound/Waves/MSS_TidalCrane_Bunker_Scan
inputs [
 UE.Attenuation.Distance {type=Float}
 UE.Source.OnPlay {type=Trigger}
]
interfaces [UE.Attenuation:1.0, UE.OutputFormat.Stereo:1.0, UE.Source.OneShot:1.0, UE.Source:1.1]
outputs [
 "UE.OutputFormat.Stereo.Audio:0" {type=Audio}
 "UE.OutputFormat.Stereo.Audio:1" {type=Audio}
 UE.Source.OneShot.OnFinished {type=Trigger}
]
pages [Default {
  input_defaults {UE.Attenuation.Distance=0.16, UE.Source.OnPlay=null}
  nodes [WavePlayer_0 = UE.Wave Player:Stereo(Play=^UE.Source.OnPlay) {also {defaults {Wave Asset=&/TidalCrane_Door_Activation}}}]
  links [
   {dst="@UE.Source.OneShot.OnFinished", src=WavePlayer_0.On Finished}
   {dst="@UE.OutputFormat.Stereo.Audio:0", src=WavePlayer_0.Out Left}
   {dst="@UE.OutputFormat.Stereo.Audio:1", src=WavePlayer_0.Out Right}
  ]
 }]
properties {
 assetImportData=&self:AssetImportData
 attenuationSettings=&/ATT_TidalCrane_Bunker_Activation_Scanner
 maxDistance=6000
 numChannels=2
 outputFormat=Stereo
 sampleRate=48000
 soundClassObject=/Game/Sounds/FortSoundClassesAndModes/Fort_08_Player_Interacts
 volume=2
 concurrencyOverrides {maxCount=2}
}
