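# One-shot stereo MetaSound source for the Tidal Crane bunker's scan speaker voice-over, judging by its name.
# On play, a random get picks an entry from a wave array variable and feeds it to a stereo wave player, whose completion signals finished.
# It uses bunker attenuation and the player-interacts sound class, caps concurrency, and sends to an ambient-volume sidechain bus.
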MetaSoundSource /MidMatchWeaponBunkerGameplay/Items/WeaponBunkerDoor/MMOBunkers_Ch6S4/Sound/Waves/MSS_TidalCrane_Bunker_Scan_Speaker_VO
inputs [
 UE.Attenuation.Distance {type=Float}
 UE.Source.OnPlay {type=Trigger}
]
interfaces [UE.Attenuation:1.0, UE.OutputFormat.Stereo:1.0, UE.Source.OneShot:1.0, UE.Source:1.1]
outputs [
 "UE.OutputFormat.Stereo.Audio:0" {type=Audio}
 "UE.OutputFormat.Stereo.Audio:1" {type=Audio}
 UE.Source.OneShot.OnFinished {type=Trigger}
]
pages [Default {
  input_defaults {UE.Attenuation.Distance=0.16, UE.Source.OnPlay=null}
  nodes [
   ArrayRandomGet_0 = Array.Random Get:WaveAsset:Array(Next=^UE.Source.OnPlay, In Array=VariableAccessor.WaveAsset:Array(Variable=InitVariable.WaveAsset:Array().Variable).Value)
   WavePlayer_0 = UE.Wave Player:Stereo(Play=@ArrayRandomGet_0.On Next, Wave Asset=@ArrayRandomGet_0.Value)
  ]
  links [
   {dst="@UE.Source.OneShot.OnFinished", src=WavePlayer_0.On Finished}
   {dst="@UE.OutputFormat.Stereo.Audio:0", src=WavePlayer_0.Out Left}
   {dst="@UE.OutputFormat.Stereo.Audio:1", src=WavePlayer_0.Out Right}
  ]
 }]
properties {
 assetImportData=&self:AssetImportData
 attenuationSettings=/MidMatchWeaponBunkerGameplay/Sound/Att_S29_Bunker
 maxDistance=34200
 numChannels=2
 outputFormat=Stereo
 sampleRate=48000
 soundClassObject=/Game/Sounds/FortSoundClassesAndModes/Fort_08_Player_Interacts
 volume=2
 busSends [{audioBus=&AB_Sidechain_FortAmbient_Vol, sendLevel=0.5}]
 concurrencyOverrides {maxCount=2}
}
